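# Additive, unlit, two-sided FX material for the AventurineAnimated car wheel cosmetic, animating glowing "Notes" and "Pikes" layers.
# Notes pan with switchable direction, with mask channels offsetting their V coordinate and cutting out regions.
# Pikes multiply two opposite-panning overlay samples, then tint them.
# Time-sampled noise builds low and high frequency signals, alongside vertex-alpha blends and a musical BeatLifetime cosine pulse, with no material output connections shown.
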Material /VehicleCosmetics/Wheels/AventurineAnimated/Materials/M_Wheel_AventurineAnimated_FX
graph {
 properties {
  bAutomaticallySetUsageInEditor=false
  bUsedWithMeshDeformer=true
  bUsedWithSkeletalMesh=true
  blendMode=BLEND_Additive
  shadingModel=MSM_Unlit
  twoSided=true
  usedShadingModels=Unlit
 }
 nodes [
  n1 = NamedRerouteUsage()
  n2 = TextureCoordinate()
  n3 = Add(
   A=StaticSwitchParameter(True=@n1, False=Constant()) {parameterName=UsePanneroffset for Notes}
   B=Multiply(A=Time(), B=ScalarParameter() {defaultValue=1, parameterName=Glow_PanSpeedAdjust})
  )
  n4 = Add(
   A=Multiply(A=@n1) {constB=-0.12}
   B=StaticSwitchParameter(True=Panner(Coordinate=@n2, Time=@n3) {speedX=0.5}, False=Panner(Coordinate=@n2, Time=@n3) {speedX=-0.5}) {defaultValue=true, parameterName=Reverse Notes}
  )
  Add(
   A=@n4
   B=AppendVector(A=ScalarParameter() {defaultValue=1, parameterName=XOffset}, B=ScalarParameter() {defaultValue=1.5, parameterName=YOffset})
  )
  n5 = ScalarParameter() {defaultValue=3, parameterName=Glow_EmissiveMultiply}
  n6 = NamedRerouteUsage()
  n7 = VectorParameter() {parameterName=Overlay Color Notes, defaultValue=<rgba 0.3231 1 1 1>}
  n8 = OneMinus(None=VertexColor().A)
  LinearInterpolate(A=Multiply(A=NamedRerouteUsage(), B=@n5), B=Multiply(A=Multiply(A=@n6, B=@n7.RGB), B=@n5), Alpha=@n8)
  n9 = NamedRerouteUsage()
  LinearInterpolate(A=Multiply(A=NamedRerouteUsage(), B=@n9), B=Multiply(A=NamedRerouteUsage(), B=@n9), Alpha=@n8)
  n10 = CollectionParameter() {collection=/Game/Athena/Events/AudioAnalysis/MPC_Musical_Parameters, parameterName=BeatLifetime}
  n11 = Multiply(A=Add(A=Cosine(None=@n10))) {constB=0.5}
  Multiply(A=Multiply(A=Power(Base=@n11) {constExponent=3}, B=Constant3Vector() {constant=<rgba 0 0 1 0>}), B=Constant() {r=10})
  n12 = Add(
   A=TextureCoordinate() {uTiling=2}
   B=AppendVector(A=ScalarParameter() {defaultValue=1, parameterName=PikesXOffset}, B=ScalarParameter() {defaultValue=0.95, parameterName=PikesYOffset})
  )
  n13 = Multiply(A=Time(), B=ScalarParameter() {defaultValue=1, parameterName=Overlay pannerspeed})
  n14 = TextureObjectParameter() {
   parameterName=Pikes
   samplerType=SAMPLERTYPE_LinearColor
   texture=/VehicleCosmetics/Wheels/AventurineAnimated/Textures/T_AventurineAnimated_Radiant_Overlay
  }
  n15 = TextureSample(UVs=Panner(Coordinate=@n12, Time=@n13) {speedX=0.2}, Tex=@n14) {
   samplerSource=SSM_Wrap_WorldGroupSettings
   samplerType=SAMPLERTYPE_LinearColor
   texture=&T_AventurineAnimated_Radiant_Overlay
  }
  n16 = TextureSample(UVs=Panner(Coordinate=@n12, Time=@n13) {speedX=-0.2}) {
   samplerSource=SSM_Wrap_WorldGroupSettings
   samplerType=SAMPLERTYPE_LinearColor
   texture=&T_AventurineAnimated_Radiant_Overlay
  }
  n17 = Power(Base=Multiply(A=@n15.RGB, B=@n16.RGB), Exp=ScalarParameter() {defaultValue=1, parameterName=Pikes Contrast})
  NamedRerouteDeclaration(None=@n17) {name=Pikes Alpha, nodeColor=<rgba 0 0.4969 1 1>}
  NamedRerouteDeclaration(None=Multiply(A=@n17, B=VectorParameter() {parameterName=Overlay Color Pikes, defaultValue=<rgba 0.0999 1 1 1>}.RGB)) {name=Pikes BC, nodeColor=<rgba 0 0.4969 1 1>}
  n18 = Multiply(A=Time(), B=ScalarParameter() {defaultValue=1, parameterName=TimeSpeedLOW Freq})
  n19 = Clamp(None=Multiply(
    A=TextureSample(UVs=@n18) {samplerType=SAMPLERTYPE_LinearColor, texture=/Game/Global/Textures/Default/Tileable_Patterns/T_PerlinNoise_Uncompressed}.R
    B=ScalarParameter() {defaultValue=1, parameterName=RandomLOWFreq}
   ))
  NamedRerouteDeclaration(None=Clamp(None=Clamp(None=@n19) {minDefault=0.3}) {maxDefault=1.5}) {name=LOW Freq out Clamp, nodeColor=<rgba 1 0.5156 0 1>}
  n20 = TextureObjectParameter() {parameterName=Albedo Notes, texture=/VehicleCosmetics/Wheels/AventurineAnimated/Textures/T_AventurineAnimated_Radiant}
  n21 = TextureSample(Tex=@n20) {samplerSource=SSM_Wrap_WorldGroupSettings, texture=&T_AventurineAnimated_Radiant}
  NamedRerouteDeclaration(None=Add(A=NamedRerouteUsage(), B=@n21.RGB)) {name=Pikes ADD BC, nodeColor=<rgba 0 0.4969 1 1>}
  NamedRerouteDeclaration(None=Add(A=NamedRerouteUsage(), B=@n21.A)) {name=Pikes ADD Alpha, nodeColor=<rgba 0 0.4969 1 1>}
  n22 = ComponentMask(None=@n4) {g=true}
  n23 = TextureObjectParameter() {
   parameterName=Masks
   samplerType=SAMPLERTYPE_LinearColor
   texture=/VehicleCosmetics/Wheels/AventurineAnimated/Textures/T_AventurineAnimated_Radiant_Mask
  }
  n24 = TextureSample(UVs=@n4, Tex=@n23) {
   samplerSource=SSM_Wrap_WorldGroupSettings
   samplerType=SAMPLERTYPE_LinearColor
   texture=&T_AventurineAnimated_Radiant_Mask
  }
  n25 = Multiply(A=@n24.B, B=NamedRerouteUsage()) {constA=0.2}
  n26 = Multiply(A=AppendVector(A=ComponentMask(None=@n4) {r=true}, B=Subtract(A=@n22, B=@n25)), B=NamedRerouteUsage())
  n27 = TextureSample(UVs=@n26, Tex=@n20) {samplerSource=SSM_Wrap_WorldGroupSettings, texture=&T_AventurineAnimated_Radiant}
  n28 = TextureSample(UVs=@n26, Tex=@n23) {
   samplerSource=SSM_Wrap_WorldGroupSettings
   samplerType=SAMPLERTYPE_LinearColor
   texture=&T_AventurineAnimated_Radiant_Mask
  }
  n29 = OneMinus(None=@n28.G)
  NamedRerouteDeclaration(None=Multiply(A=@n27.RGB, B=@n29)) {name=Notes Mask, nodeColor=<rgba 1 0 0.5 1>}
  NamedRerouteDeclaration(None=Multiply(A=@n27.A, B=@n29)) {name=Notes Alpha, nodeColor=<rgba 1 0 0.5 1>}
  NamedRerouteDeclaration(None=Multiply(B=@n19) {constA=0.2}) {name=LOW Freq out Mult, nodeColor=<rgba 1 0.5156 0 1>}
  n30 = Multiply(A=Time(), B=ScalarParameter() {defaultValue=0.1, parameterName=TimeSpeedHighFreq})
  n31 = TextureSample(UVs=@n30) {samplerType=SAMPLERTYPE_LinearColor, texture=/Game/Weapons/FORT_Melee/Pickaxe_RhombCamo/FX/T_Pickaxe_RhombCamo_Perlin}
  n32 = ScalarParameter() {defaultValue=10, parameterName=RandomHighFreq}
  NamedRerouteDeclaration(None=Clamp(None=Multiply(A=@n31.R, B=@n32))) {name=HIGH Freq out Clamp, nodeColor=<rgba 1 0.5156 0 1>}
  n33 = Saturate(None=ScalarParameter() {defaultValue=1, parameterName=NotesUVDirection_X})
  n34 = Saturate(None=ScalarParameter() {defaultValue=1, parameterName=NotesUVDirection_Y})
  NamedRerouteDeclaration(None=AppendVector(A=LinearInterpolate(Alpha=@n33) {constA=-1}, B=LinearInterpolate(Alpha=@n34) {constA=-1})) {name=NotesUVDirection, nodeColor=<rgba 1 0.4453 0 1>}
  Sine(None=@n30)
  Sine(None=@n18)
 ]
}
